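# Stereo MetaSound source that plays the looping PulseLancer BubbleShield music track for ValidCandy.
# On play, a looping wave player feeds both output channels and writes its On Looped trigger to a trigger variable.
# An Int32 variable and BeatCount output are declared but left unwired.
# It uses the Fort gameplay music sound class and sends to the ValidCandy music audio bus, with base submix and submix sends disabled.
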MetaSoundSource /ValidCandy/Assets/Audio/IP_MSS_ValidCandy_Loop
inputs [UE.Source.OnPlay {type=Trigger}]
interfaces [UE.OutputFormat.Stereo:1.0, UE.Source:1.1]
outputs [
 BeatCount {type=Int32}
 "UE.OutputFormat.Stereo.Audio:0" {type=Audio}
 "UE.OutputFormat.Stereo.Audio:1" {type=Audio}
]
pages [Default {
  input_defaults {UE.Source.OnPlay=null}
  nodes [
   InitVariable.Int32()
   WavePlayer_0 = UE.Wave Player:Stereo(Play=^UE.Source.OnPlay) {also {
     defaults {Loop="<MetaSoundBuilderResult.SUCCEEDED: 0>", Wave Asset=&/IP_MX_S37_PulseLancer_BubbleShield_LP}
     version [1,1]
    }}
   VariableMutator.Trigger(Variable=InitVariable.Trigger().Variable, Value=@WavePlayer_0.On Looped)
  ]
  links [{dst="@UE.OutputFormat.Stereo.Audio:0", src=WavePlayer_0.Out Left}, {dst="@UE.OutputFormat.Stereo.Audio:1", src=WavePlayer_0.Out Right}]
 }]
properties {
 assetImportData=&self:AssetImportData
 bEnableBaseSubmix=false
 bEnableSubmixSends=false
 bLooping=true
 duration=10000
 numChannels=2
 outputFormat=Stereo
 sampleRate=48000
 soundClassObject=/Game/Sounds/FortSoundClassesAndModes/Fort_Music_Gameplay
 assetUserData [&self:AudioAssetUserData_0]
 busSends [{audioBus=&/AB_ValidCandy_Music, sendLevel=1}]
 soundSubmixSends [{}]
}
