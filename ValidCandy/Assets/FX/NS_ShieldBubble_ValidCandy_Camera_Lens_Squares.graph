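# Niagara system for the ValidCandy shield bubble effect, named for a camera lens squares element.
# It holds a single Camera_Lens_Squares emitter, with no renderers listed, and SystemState is the only module recorded.
# User parameters PrimaryColour and SecondaryColour expose warm orange default tints, though the file does not show how the emitter uses them.
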NiagaraSystem /ValidCandy/Assets/FX/NS_ShieldBubble_ValidCandy_Camera_Lens_Squares
dependencies {usedDataInterfaces [], usedDynamicInputs [], usedModules [&SystemState]}
emitters {Camera_Lens_Squares {rendererClasses [], renderers []}}
summary {
 systemName=NS_ShieldBubble_ValidCandy_Camera_Lens_Squares
 emitters [{emitterName=Camera_Lens_Squares, rendererClasses []}]
 userVariables [
  User.PrimaryColour {type=LinearColor, description="", defaultValue {struct=LinearColor, value=<rgba 1 0.4351 0.0753 1>}}
  User.SecondaryColour {type=LinearColor, description="", defaultValue {struct=LinearColor, value=<rgba 1 0.6786 0.3408 1>}}
 ]
}
