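# Niagara system for a shield bubble passthrough flash in ValidCandy, judging by its name.
# The Light emitter burst-spawns a short-lived light particle in the user secondary colour, with a curve fading its boosted alpha to zero.
# Six other emitters for core flares, comet and castoff meshes, and velocity squares list no renderers or modules here.
# User parameters expose colours, flare lifetime and spawn settings, though only the secondary colour is visibly used.
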NiagaraSystem /ValidCandy/Assets/FX/NS_ShieldBubble_ValidCandy_PassthroughFlash
dependencies {
 usedDataInterfaces [&NiagaraDataInterfaceCurve]
 usedDynamicInputs [&Multiply_VectorByFloat, &FN_LOD_Distance_Remap]
 usedModules [
  &self:SystemSpawnScript.SystemScriptSource.SystemScriptGraph.NiagaraNodeAssignment_1.SetVariables_#1
  &SystemState
 ]
}
emitters {
 CometMesh {rendererClasses [], renderers []}
 CoreFlare_Horz {rendererClasses [], renderers []}
 CoreFlare_Vert {rendererClasses [], renderers []}
 Light {
  emitterUpdateScript [
   &EmitterState()
   &SpawnBurst_Instantaneous(Use Spawn Probability=false, UseLoopCountLimit=false)
  ]
  particleSpawnScript [&InitializeParticle(Color=$User.SecondaryColour, Lifetime=0.6, UsePositionOffset=false, Write Material Random=true)]
  particleUpdateScript [
   &ParticleState()
   &ScaleColor(
    Scale Alpha=&FloatFromCurve(FloatCurve=DataInterface() {propertyValues {Curve {keys [<at 0 50>,<at 1 0>]}}})
    ScaleA=true
    ScaleRGB=true
   )
  ]
  rendererClasses [&NiagaraLightRendererProperties]
  renderers [{rendererClass=&NiagaraLightRendererProperties, rendererIndex=0}]
 }
 Mesh_Castoff {rendererClasses [], renderers []}
 Vel_Squares_A {rendererClasses [], renderers []}
 Vel_Squares_B {rendererClasses [], renderers []}
}
summary {
 systemName=NS_ShieldBubble_ValidCandy_PassthroughFlash
 emitters [
  {emitterName=CoreFlare_Vert, rendererClasses []}
  {emitterName=CoreFlare_Horz, rendererClasses []}
  {emitterName=Mesh_Castoff, rendererClasses []}
  {emitterName=Vel_Squares_A, rendererClasses []}
  {emitterName=Vel_Squares_B, rendererClasses []}
  {emitterName=CometMesh, rendererClasses []}
  {emitterName=Light, rendererClasses [&NiagaraLightRendererProperties]}
 ]
 userVariables [
  User.FlareLIfetime {type=NiagaraFloat, defaultValue=0.6, description=""}
  User.PrimaryColour {type=LinearColor, description="", defaultValue {struct=LinearColor, value=<rgba 1 0.4226 0.1201 1>}}
  User.SecondaryColour {type=LinearColor, description="", defaultValue {struct=LinearColor, value=<rgba 1 0.6786 0.3408 1>}}
  User.SpawnBurst {type=NiagaraInt32, defaultValue=1}
  User.SpawnRate {type=NiagaraFloat, defaultValue=50, description=""}
 ]
}
