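# Niagara system for the ValidCandy shield bubble "TimesUp" effect, which by its name plays when the shield's time runs out.
# It combines seven emitters named for a core flare, mesh castoff, looping rays, burst wave, small sparks, bokeh and energy burst.
# User parameters expose warm primary and secondary colours plus an integer Stage.
# System scripts set variables alongside SystemState, and dynamic inputs include distance-based LOD remapping and colour scaling.
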NiagaraSystem /ValidCandy/Assets/FX/NS_ShieldBubble_ValidCandy_TimesUp
dependencies {
 usedDataInterfaces []
 usedDynamicInputs [
  &Vector2DFromFloat
  &FN_LOD_Distance_Remap
  &Multiply_LinearColorByFloat
 ]
 usedModules [
  &self:SystemSpawnScript.SystemScriptSource.SystemScriptGraph.NiagaraNodeAssignment_0.SetVariables_#1
  &SystemState
 ]
}
emitters {
 Bokeh {rendererClasses [], renderers []}
 BurstWave {rendererClasses [], renderers []}
 CoreFlare {rendererClasses [], renderers []}
 EnergyBurst {rendererClasses [], renderers []}
 Mesh_Castoff {rendererClasses [], renderers []}
 Rays_Looping {rendererClasses [], renderers []}
 Sparks_Small {rendererClasses [], renderers []}
}
summary {
 systemName=NS_ShieldBubble_ValidCandy_TimesUp
 emitters [
  {emitterName=CoreFlare, rendererClasses []}
  {emitterName=Mesh_Castoff, rendererClasses []}
  {emitterName=Rays_Looping, rendererClasses []}
  {emitterName=BurstWave, rendererClasses []}
  {emitterName=Sparks_Small, rendererClasses []}
  {emitterName=Bokeh, rendererClasses []}
  {emitterName=EnergyBurst, rendererClasses []}
 ]
 userVariables [
  User.PrimaryColour {type=LinearColor, description="", defaultValue {struct=LinearColor, value=<rgba 1 0.4003 0.1001 1>}}
  User.SecondaryColour {type=LinearColor, description="", defaultValue {struct=LinearColor, value=<rgba 1 0.6786 0.3408 1>}}
  User.Stage {type=NiagaraInt32, defaultValue=2}
 ]
}
