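# Additive, unlit, two-sided material for a spherical energy shield that can deploy, shrink, dissolve and react to hits and audio.
# Noise-distorted spherical UVs drive animated grid and flare lines, layered with Fresnel color, thin-film rainbow and depth-based intersection glow.
# Timestamped sphere masks at a hit location create expanding waves feeding emissive, distortion and vertex offset.
# Quality and shading-path switches simplify or drop costly effects on mobile and lower tiers.
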Material /ValidCandy/Assets/Materials/M_ValidCandy_SphereShield
graph {
 properties {
  bUsedWithMeshDeformer=true
  bUsedWithSkeletalMesh=true
  blendMode=BLEND_Additive
  pixelDepthOffsetMode=PDOM_Legacy
  shadingModel=MSM_Unlit
  twoSided=true
  usedShadingModels=Unlit
 }
 nodes [
  Constant3Vector() {constant=<rgba -100 -100 -100 0>}
  Constant3Vector() {constant=<rgba -100 -100 -100 0>}
  Constant()
  Constant() {r=1}
  Frac(None=Time())
  Frac(None=Time())
  Frac(None=Time())
  Frac(None=Time())
  n1 = Constant() {r=-5}
  n2 = ScalarParameter() {defaultValue=1, parameterName=FlareStripLines}
  n3 = Multiply(A=NamedRerouteUsage(), B=Constant2Vector() {g=0.1, r=1}) {constA=1}
  n4 = Panner(Coordinate=NamedRerouteUsage()) {speedY=-0.002}
  n5 = ComponentMask(None=@n4) {g=true}
  n6 = ComponentMask(None=@n4) {r=true}
  n7 = ScalarParameter() {defaultValue=0.15, parameterName=HM_Distort_Intensity}
  n8 = Multiply(A=NamedRerouteUsage(), B=@n7) {constB=0.15}
  n9 = NamedRerouteUsage()
  n10 = Add(
   A=MaterialFunctionCall(X=Add(A=@n5, B=@n6), Y=Subtract(A=@n5, B=@n6)) {materialFunction=&MakeFloat2}.Result
   B=Add(A=Add(A=@n8, B=@n9), B=NamedRerouteUsage())
  )
  n11 = Divide(A=Multiply(A=@n10) {constB=4}, B=ScalarParameter() {defaultValue=0.1, group=Hummoon, parameterName=HMLineScale})
  n12 = Add(A=@n11, B=@n3) {constB=0.001}
  n13 = Constant2Vector() {g=1, r=0.5}
  n14 = Multiply(A=Frac(None=Multiply(A=@n12, B=@n13))) {constB=2}
  n15 = MaterialFunctionCall(Float2=Abs(None=Subtract(A=@n14))) {materialFunction=&BreakOutFloat2Components}
  n16 = Multiply(A=@n2, B=@n15.R)
  n17 = NamedRerouteUsage()
  n18 = Constant()
  n19 = Add(
   A=TextureSample!(UVs=Panner(Coordinate=@n16) {speedX=-0.1}, Tex=@n17, "DDX(UVs)"=@n18, "DDY(UVs)"=@n18, Apply View MipBias) {mipValueMode=TMVM_Derivative, samplerSource=SSM_Wrap_WorldGroupSettings}.R
   B=TextureSample(UVs=Panner(Coordinate=@n16) {speedX=0.1}, Tex=@n17, "DDX(UVs)"=@n18, "DDY(UVs)"=@n18) {mipValueMode=TMVM_Derivative, samplerSource=SSM_Wrap_WorldGroupSettings}.R
  )
  n20 = OneMinus(None=SmoothStep(Min=@n1, Value=Add(A=@n19, B=@n15.G)) {constMin=-5})
  n21 = Add(A=LinearInterpolate(Alpha=NamedRerouteUsage()) {constB=0.3}, B=Constant() {r=1}) {constB=0.1}
  n22 = Divide(A=@n21, B=@n15.G) {constA=0.01, constB=1}
  n23 = Multiply(A=Frac(None=Multiply(A=@n11, B=@n13))) {constB=2}
  n24 = MaterialFunctionCall(Float2=Abs(None=Subtract(A=@n23))) {materialFunction=&BreakOutFloat2Components}
  n25 = Multiply(A=@n2, B=@n24.R)
  n26 = NamedRerouteUsage()
  n27 = Constant()
  n28 = Add(
   A=TextureSample(UVs=Panner(Coordinate=@n25) {speedX=-0.1}, Tex=@n26, "DDX(UVs)"=@n27, "DDY(UVs)"=@n27) {mipValueMode=TMVM_Derivative, samplerSource=SSM_Wrap_WorldGroupSettings}.R
   B=TextureSample(UVs=Panner(Coordinate=@n25) {speedX=0.1}, Tex=@n26, "DDX(UVs)"=@n27, "DDY(UVs)"=@n27) {mipValueMode=TMVM_Derivative, samplerSource=SSM_Wrap_WorldGroupSettings}.R
  )
  n29 = OneMinus(None=SmoothStep(Min=@n1, Value=Add(A=@n28, B=@n24.G)) {constMin=-5})
  n30 = Add(A=@n29, B=Multiply(A=@n29, B=Divide(A=@n21, B=@n24.G) {constA=0.01, constB=1}))
  n31 = Subtract(A=@n11, B=@n3) {constB=0.001}
  n32 = Multiply(A=Frac(None=Multiply(A=@n31, B=@n13))) {constB=2}
  n33 = MaterialFunctionCall(Float2=Abs(None=Subtract(A=@n32))) {materialFunction=&BreakOutFloat2Components}
  n34 = Multiply(A=@n2, B=@n33.R)
  n35 = NamedRerouteUsage()
  n36 = Constant()
  n37 = Add(
   A=TextureSample(UVs=Panner(Coordinate=@n34) {speedX=-0.1}, Tex=@n35, "DDX(UVs)"=@n36, "DDY(UVs)"=@n36) {mipValueMode=TMVM_Derivative, samplerSource=SSM_Wrap_WorldGroupSettings}.R
   B=TextureSample(UVs=Panner(Coordinate=@n34) {speedX=0.1}, Tex=@n35, "DDX(UVs)"=@n36, "DDY(UVs)"=@n36) {mipValueMode=TMVM_Derivative, samplerSource=SSM_Wrap_WorldGroupSettings}.R
  )
  n38 = OneMinus(None=SmoothStep(Min=@n1, Value=Add(A=@n37, B=@n33.G)) {constMin=-5})
  n39 = Divide(A=@n21, B=@n33.G) {constA=0.01, constB=1}
  n40 = MaterialFunctionCall(X=Add(A=@n20, B=Multiply(A=@n20, B=@n22)), Y=@n30, Z=Add(A=@n38, B=Multiply(A=@n38, B=@n39))) {materialFunction=&MakeFloat3}
  n41 = Abs(None=MaterialFunctionCall!(Low=@n40.Result, High=@n40.Result, Epic=@n40.Result, --------, Mobile=@n30, Medium=@n30) {materialFunction=/FortniteBaseMaterials/Functions/MF_QualitySwitch_Vector}.Result)
  n42 = Multiply(
   A=Desaturation(None=Power(Base=@n41), Fraction=Constant() {r=0.5})
   B=ScalarParameter() {defaultValue=1, group=Hummoon, parameterName=Flare_Intensity}
  ) {constB=10}
  Multiply(A=@n42, B=Constant3Vector() {constant=<rgba 0.276 0.1513 0.0359 1>})
  n43 = VertexNormalWS()
  Multiply(
   A=MaterialFunctionCall!(Vector=@n43) {materialFunction=/Engine/Functions/Engine_MaterialFunctions01/Coordinates/LongLatToUV}.Result
   B=Constant2Vector() {g=1, r=-1}
  )
  n44 = NamedRerouteUsage()
  n45 = NamedRerouteUsage()
  n46 = NamedRerouteUsage()
  n47 = Multiply(A=Power(Base=NamedRerouteUsage()), B=@n46)
  n48 = LinearInterpolate(A=Multiply(A=NamedRerouteUsage()) {constB=0.5}, B=Multiply(A=Power(Base=@n45)) {constB=0.1}, Alpha=Saturate(None=Add(A=@n47, B=@n46)))
  n49 = Divide(A=Add(A=PixelDepth().R, B=Constant() {r=-500}), B=ScalarParameter() {defaultValue=6500, group=Opacity, parameterName=Distance_Fade})
  n50 = Saturate(None=TwoSidedSign())
  n51 = Max(A=NamedRerouteUsage(), B=NamedRerouteUsage())
  n52 = Saturate(None=Divide(A=LinearInterpolate(Alpha=@n50) {constA=0.001, constB=0.01}, B=NamedRerouteUsage()) {constA=0.001})
  n53 = Divide(
   A=LinearInterpolate(B=LinearInterpolate(Alpha=@n50) {constA=0.02, constB=0.05}, Alpha=Max(A=@n51, B=@n52)) {constA=0.007, constB=0.02}
   B=NamedRerouteUsage()
  ) {constA=0.01}
  n54 = Saturate(None=@n53)
  n55 = MaterialFunctionCall!(Low=@n53, High=@n53, Epic=@n53, --------, Mobile=@n54, Medium=@n54) {materialFunction=/FortniteBaseMaterials/Functions/MF_QualitySwitch_Scalar}
  n56 = Saturate(None=@n55.Result)
  n57 = NamedRerouteUsage()
  n58 = Multiply(
   A=OneMinus(None=Saturate(None=@n49))
   B=LinearInterpolate(A=NamedRerouteUsage(), B=LinearInterpolate(A=@n56, B=@n55.Result, Alpha=@n57), Alpha=NamedRerouteUsage())
  )
  n59 = VectorParameter() {group=HitWave, parameterName=HitWave-Loc, defaultValue=<rgba 100 -1400 -150 0>}
  n60 = ScalarParameter() {defaultValue=300, group=HitWave, parameterName=HitWave-SizeMultiplierOuterRing}
  n61 = Multiply(A=@n55.Result) {constB=0.5}
  n62 = NamedRerouteUsage()
  n63 = Multiply(A=Saturate(None=Subtract(A=@n55.Result, B=@n62) {constB=0.5})) {constB=10, desc=Hit Grid Brightness}
  n64 = ScalarParameter() {defaultValue=1, parameterName=HitEffectLerp}
  n65 = Subtract(A=Time(), B=ScalarParameter() {defaultValue=1, desc=Time-stamp, group=HitWave, parameterName=GameTime})
  n66 = ScalarParameter() {defaultValue=0.5, group=HitWave, parameterName=HitWave-DurationNew}
  n67 = OneMinus(None=Saturate(None=Divide(A=@n65, B=@n66)))
  n68 = Multiply(
   A=SphereMask(
    A=WorldPosition() {worldPositionShaderOffset=WPT_ExcludeAllShaderOffsets}.XYZ
    B=ComponentMask(None=@n59.RGB) {b=true, g=true, r=true}
    Radius=Divide(A=@n60) {constB=2.5}
   ) {attenuationRadius=128, hardnessPercent=0}
   B=Multiply(A=LinearInterpolate(A=@n61, B=@n63, Alpha=@n64), B=Power(Base=@n67) {constExponent=4})
  )
  n69 = NamedRerouteUsage()
  n70 = ScalarParameter() {defaultValue=2, group=Grid, parameterName=HunmoonLine Polar Fade}
  n71 = ScalarParameter() {defaultValue=1, group=Grid, parameterName=HunmoonLine Intensity}
  n72 = NamedRerouteUsage()
  n73 = NamedRerouteUsage()
  n74 = Multiply(
   A=Multiply(A=Add(A=@n58, B=@n68), B=Power(Base=@n69, Exp=@n70))
   B=LinearInterpolate(A=@n71, B=Multiply(A=@n71) {constB=4}, Alpha=Multiply(A=@n72, B=@n73)) {constB=10}
  )
  n75 = NamedRerouteUsage()
  n76 = ScalarParameter() {defaultValue=0.2, parameterName=BubbleThinFilmFresnelPower}
  n77 = LinearInterpolate(
   A=LinearInterpolate(
    A=VectorParameter() {group=Grid, parameterName=HunmoonLine Color, defaultValue=<rgba 0.8 0.8 0.8 1>}.RGB
    B=NamedRerouteUsage()
    Alpha=Power(Base=@n75, Exp=@n76)
   )
   B=VectorParameter() {group=HitWave, parameterName=HitWave-Tint, defaultValue=<rgba 0.5 0.5 0.5 1>}.RGB
   Alpha=NamedRerouteUsage()
  )
  n78 = NamedRerouteUsage()
  n79 = Multiply(A=@n55.Result, B=NamedRerouteUsage())
  n80 = Multiply(
   A=VectorParameter() {group=HitWave, parameterName=HitWave-Tint, defaultValue=<rgba 0.5 0.5 0.5 1>}.RGB
   B=NamedRerouteDeclaration(None=Saturate(None=@n79)) {name=Hommon_HitWave, nodeColor=<rgba 1 0 0.0469 1>}
  )
  n81 = ScalarParameter() {defaultValue=1, parameterName=HitEffectLerp}
  n82 = OneMinus(None=NamedRerouteUsage())
  n83 = LinearInterpolate(B=Multiply(A=Multiply(A=@n78, B=@n80) {constB=15}, B=LinearInterpolate(Alpha=@n81) {constB=85}) {constB=85}, Alpha=@n82)
  n84 = Constant()
  n85 = Add(
   A=QualitySwitch(
    Default=ShadingPathSwitch(Default=@n83, Deferred=@n83, Mobile=@n84) {desc=Migrated from FeatureLevelSwitch}
    Low=@n84
    High=@n83
    Medium=@n84
   )
   B=@n80
  )
  n86 = Multiply(
   A=Multiply(
    A=Power(Base=@n82) {constExponent=4}
    B=VectorParameter() {group=HitWave, parameterName=Hitwave-AddBrighness, defaultValue=<rgba 0.1 0.1 0.1 1>}.RGB
   )
   B=NamedRerouteUsage()
  )
  n87 = NamedRerouteUsage()
  n88 = Add(
   A=Add(A=Add(A=@n44, B=@n48) {desc=+ intersection edge}, B=Multiply(A=@n74, B=@n77))
   B=Add(A=Saturate(None=@n85), B=Multiply(A=@n86, B=@n87))
  )
  n89 = NamedRerouteUsage()
  NamedRerouteDeclaration(None=MaterialFunctionCall(Emissive=Add(A=@n88, B=@n89)) {materialFunction=&MF_FX_EmissiveAdjustments}.Result) {name=OutEmssive, nodeColor=<rgba 0.5 1 0 1>}
  NamedRerouteDeclaration(None=ScalarParameter() {defaultValue=1, group=Audio Reactivity, parameterName=EnableAudioReactivity}) {name=ReactivityOn, nodeColor=<rgba 0.1719 0 1 1>}
  n90 = VertexNormalWS()
  n91 = CameraVectorWS()
  n92 = SceneDepth()
  n93 = Add(A=PixelDepth().R, B=ScalarParameter() {group=IntersectingEdge, parameterName=IntersectingEdgeBias})
  n94 = Multiply(
   A=Abs(None=DotProduct(A=@n90, B=@n91))
   B=Divide(A=Subtract(A=@n92.R, B=@n93), B=ScalarParameter() {defaultValue=200, group=IntersectingEdge, parameterName=IntersectingEdgeDivisor})
  )
  n95 = Power(
   Base=OneMinus(None=Saturate(None=@n94))
   Exp=ScalarParameter() {defaultValue=3, group=IntersectingEdge, parameterName=Intersection Edge Exponent}
  )
  NamedRerouteDeclaration(None=@n95) {name=IntersectingEdge_Mask, nodeColor=<rgba 0 0.2031 1 1>}
  n96 = Abs(None=@n24.G)
  n97 = ComponentMask(None=Panner(Coordinate=@n10) {speedX=-0.1}) {r=true}
  n98 = Multiply(A=Frac(None=Multiply(A=@n97) {constB=2})) {constB=2}
  n99 = OneMinus(None=Abs(None=Subtract(A=@n98)))
  n100 = NamedRerouteUsage()
  n101 = LinearInterpolate(Alpha=@n100) {constA=2, constB=0.1}
  n102 = LinearInterpolate(Alpha=Multiply(A=@n100) {constB=10}) {constA=1, constB=0.5}
  n103 = Saturate(None=TwoSidedSign())
  n104 = Multiply(A=@n99, B=Divide(A=LinearInterpolate(A=@n101, B=@n102, Alpha=@n103) {constB=0.01}, B=@n21) {constA=0, constB=1})
  n105 = Power(Base=OneMinus(None=Add(A=@n99, B=@n104))) {constExponent=3}
  NamedRerouteDeclaration(None=Saturate(None=Multiply(A=@n96, B=@n105))) {name=HM_LineThicknessFactor_Scan, nodeColor=<rgba 0.0547 1 0 1>}
  NamedRerouteDeclaration(None=@n96) {name=HM_LinesMask, nodeColor=<rgba 1 0 0.375 1>}
  n106 = Multiply(A=NamedRerouteUsage(), B=Constant2Vector() {g=10, r=20}) {constB=10}
  n107 = Constant()
  n108 = TextureSample(UVs=Panner(Coordinate=@n106) {speedX=-0.2, speedY=0.1}, "DDX(UVs)"=@n107, "DDY(UVs)"=@n107) {mipValueMode=TMVM_Derivative, samplerSource=SSM_Wrap_WorldGroupSettings, texture=/Game/Global/Textures/Default/Noise/T_godray_Noise}
  n109 = NamedRerouteUsage()
  n110 = Multiply(A=@n108.R, B=@n109)
  NamedRerouteDeclaration(None=Add(A=Power(Base=@n110) {constExponent=1}, B=Multiply(A=@n109) {constB=0.5})) {name=HM_LinesMask_Noise, nodeColor=<rgba 0 0.6953 1 1>}
  n111 = Constant()
  n112 = ScalarParameter() {defaultValue=300, group=HitWave, parameterName=HitWave-SizeMultiplierOuterRing}
  n113 = Time()
  n114 = ScalarParameter() {defaultValue=1, desc=Time-stamp, group=HitWave, parameterName=GameTime}
  n115 = Saturate(None=Divide(A=Subtract(A=@n113, B=@n114), B=ScalarParameter() {defaultValue=0.5, group=HitWave, parameterName=HitWave-DurationNew}))
  n116 = Power(Base=@n115, Exp=ScalarParameter() {defaultValue=0.5, group=HitWave, parameterName=HitWave-OuterRingDelay}) {constExponent=0.7}
  n117 = ComponentMask(None=VectorParameter() {group=HitWave, parameterName=HitWave-Loc, defaultValue=<rgba 100 -1400 -150 0>}.RGB) {b=true, g=true, r=true}
  n118 = WorldPosition() {worldPositionShaderOffset=WPT_ExcludeAllShaderOffsets}
  n119 = ScalarParameter() {defaultValue=-0.1, group=HitWave, parameterName=HitWave-InnerRingDelay}
  n120 = Multiply(
   A=Saturate(None=LinearInterpolate(A=@n119, Alpha=@n115) {constA=-0.7})
   B=ScalarParameter() {defaultValue=1000, group=HitWave, parameterName=HitWave-SizeMultiplierInnerRing}
  )
  n121 = ScalarParameter() {defaultValue=0.25, group=HitWave, parameterName=Hitwave-SphereHard2}
  n122 = Multiply(
   A=SphereMask(
    A=@n118.XYZ
    B=@n117
    Radius=Multiply(A=@n112, B=@n116)
    Hardness=ScalarParameter() {defaultValue=0.7, group=HitWave, parameterName=Hitwave-SphereHard1}
   ) {hardnessPercent=0}
   B=OneMinus(None=SphereMask(A=@n118.XYZ, B=@n117, Radius=@n120, Hardness=@n121) {hardnessPercent=0})
  )
  NamedRerouteDeclaration(None=QualitySwitch(
    Default=ShadingPathSwitch(Default=@n122, Deferred=@n122, Mobile=@n111) {desc=Migrated from FeatureLevelSwitch}
    Low=@n111
    High=@n122
    Medium=@n111
   )) {name=HitWave, nodeColor=<rgba 0 1 0.2734 1>}
  NamedRerouteDeclaration(None=@n115) {name=HItwave_Time, nodeColor=<rgba 0 0.3672 1 1>}
  n123 = TextureSample!(UVs=Multiply(A=NamedRerouteUsage()) {constB=2}, Tex, Apply View MipBias) {
   samplerSource=SSM_Wrap_WorldGroupSettings
   samplerType=SAMPLERTYPE_LinearGrayscale
   texture=/Game/Global/Textures/Default/Noise/T_SwirlNoise_N_height
  }
  n124 = LinearInterpolate(Alpha=@n123.R) {constA=0.8}
  n125 = Constant() {r=0.8}
  n126 = ShadingPathSwitch(Default=@n124, Deferred=@n124, Mobile=@n125) {desc=Migrated from FeatureLevelSwitch}
  n127 = NamedRerouteUsage()
  n128 = LinearInterpolate(Alpha=StaticSwitchParameter(
    True=MaterialFunctionCall() {materialFunction=&DebugTimeSine}.Result
    False=ScalarParameter() {parameterName=Deploy}
   ) {parameterName=Debug}) {constA=-1, constB=0.2}
  NamedRerouteDeclaration(None=Step(Y=QualitySwitch(Default=@n126, Low=@n125, High=@n124, Medium=@n125), X=Subtract(A=@n127, B=@n128))) {name=DeployMask, nodeColor=<rgba 0.6875 1 0 1>}
  NamedRerouteDeclaration(None=@n123.R) {name=DeployNoise, nodeColor=<rgba 0.9922 0 1 1>}
  n129 = Saturate(None=NamedRerouteUsage())
  n130 = ScalarParameter() {defaultValue=0.8, group=HitWave, parameterName=Hitwave-pow}
  n131 = Multiply(A=Saturate(None=Power(Base=@n129, Exp=@n130)), B=ScalarParameter() {defaultValue=50, group=HitWave, parameterName=WPO-AddWave2})
  n132 = VertexNormalWS()
  n133 = Constant()
  n134 = OneMinus(None=NamedRerouteUsage())
  n135 = MaterialFunctionCall() {materialFunction=&DebugTimeSine}
  n136 = ScalarParameter() {parameterName=Deploy}
  n137 = Multiply(
   A=LinearInterpolate(Alpha=NamedRerouteUsage()) {constA=0.7}
   B=OneMinus(None=StaticSwitchParameter(True=@n135.Result, False=@n136) {parameterName=Debug})
  )
  n138 = LinearInterpolate(Alpha=@n137) {constA=-1, constB=2}
  n139 = LinearInterpolate(Alpha=Saturate(None=Subtract(A=@n134, B=@n138))) {constB=70}
  n140 = QualitySwitch(
   Default=ShadingPathSwitch(Default=@n139, Deferred=@n139, Mobile=@n133) {desc=Migrated from FeatureLevelSwitch}
   Low=@n133
   High=@n139
   Medium=@n133
  )
  NamedRerouteDeclaration(None=Add(A=Multiply(A=@n131, B=@n132), B=Multiply(A=@n132, B=@n140))) {name=DeplyHItWave, nodeColor=<rgba 0 1 0.6719 1>}
  n141 = Add(A=NamedRerouteUsage(), B=ScalarParameter() {defaultValue=1, group=Color, parameterName=FresnelBrightness})
  n142 = VectorParameter() {group=Color, parameterName=ColorCenter, defaultValue=<rgba 0.2 0.2 0.2 1>}
  n143 = Fresnel(
   ExponentIn=ScalarParameter() {defaultValue=0.1, group=Color, parameterName=FresnelExp}
   BaseReflectFractionIn=ScalarParameter() {group=Color, parameterName=FresnelBase}
  )
  n144 = VectorParameter() {group=Color, parameterName=ColorEdge, defaultValue=<rgba 0.2 0.2 0.2 1>}
  n145 = LinearInterpolate(
   A=LinearInterpolate(A=@n142.RGB, B=VectorParameter() {group=Color, parameterName=ColorMid, defaultValue=<rgba 0.2 0.2 0.2 1>}.RGB, Alpha=@n143)
   B=@n144.RGB
   Alpha=Fresnel() {baseReflectFraction=0, exponent=10}
  )
  n146 = LinearInterpolate(A=@n142.RGB, B=Divide(A=@n144.RGB) {constB=4}, Alpha=@n143)
  n147 = ShadingPathSwitch(Default=@n145, Deferred=@n145, Mobile=@n146) {desc=Migrated from FeatureLevelSwitch}
  NamedRerouteDeclaration(None=Multiply(A=Divide(A=@n141) {constB=10}, B=QualitySwitch(Default=@n147, Low=@n146, High=@n145, Medium=@n146))) {name=InnerColor, nodeColor=<rgba 0 0.6719 1 1>}
  n148 = NamedRerouteUsage()
  n149 = NamedRerouteUsage()
  n150 = NamedRerouteUsage()
  n151 = Multiply(A=ComponentMask(None=Add(A=@n149, B=@n150)) {g=true})
  n152 = Subtract(A=Multiply(A=Frac(None=@n151)) {constB=2})
  n153 = Multiply(A=VertexNormalWS(), B=Multiply(A=Power(Base=@n152) {constExponent=10}) {constB=12})
  n154 = Constant()
  n155 = LinearInterpolate(
   B=QualitySwitch(
    Default=ShadingPathSwitch(Default=@n153, Deferred=@n153, Mobile=@n154) {desc=Migrated from FeatureLevelSwitch}
    Low=@n154
    High=@n153
    Medium=@n154
   )
   Alpha=NamedRerouteUsage()
  )
  n156 = VertexNormalWS()
  n157 = TextureSample() {
   samplerSource=SSM_Wrap_WorldGroupSettings
   samplerType=SAMPLERTYPE_LinearGrayscale
   texture=&T_SwirlNoise_N_height
  }
  n158 = Constant() {r=1}
  n159 = ScalarParameter() {parameterName=Shrink}
  n160 = Multiply(
   A=QualitySwitch(
    Default=ShadingPathSwitch(Default=@n157.R, Deferred=@n157.R, Mobile=@n158) {desc=Migrated from FeatureLevelSwitch}
    Low=@n158
    High=@n157.R
    Medium=@n158
   )
   B=@n159
  )
  n161 = LinearInterpolate(A=Multiply(A=Multiply(A=@n156) {constB=-1}, B=LinearInterpolate(Alpha=@n160) {constB=3000}), Alpha=@n159) {constB=0}
  n162 = Multiply(A=Multiply(A=VertexNormalWS()) {constB=-1}, B=LinearInterpolate(Alpha=@n159) {constB=800})
  NamedRerouteDeclaration(None=Add(A=Add(A=@n148, B=@n155), B=Add(A=@n161, B=@n162)) {constB=0}) {name=WPO, nodeColor=<rgba 1 0.3516 0 1>}
  n163 = ComponentMask(None=VectorParameter() {group=Grid, parameterName=GridTiling, defaultValue=<rgba 1.296 0.7776 0 0>}.RGB) {g=true, r=true}
  n164 = TextureSampleParameter2D(UVs=Multiply(A=NamedRerouteUsage(), B=Multiply(A=@n163) {constB=6})) {
   parameterName=Disolve_Testure
   samplerSource=SSM_Wrap_WorldGroupSettings
   samplerType=SAMPLERTYPE_LinearGrayscale
   texture=&T_SwirlNoise_N_height
  }
  n165 = Time()
  n166 = Frac(None=Add(A=@n164.R, B=Frac(None=@n165)))
  n167 = OneMinus(None=Multiply(A=Sine(None=@n166), B=@n164.G))
  n168 = ScalarParameter() {group=Opacity, parameterName=DissolveAmount}
  n169 = Multiply(A=ShadingPathSwitch(Default=@n167, Deferred=@n167, Forward=@n164.R, Mobile=@n164.R), B=@n168)
  n170 = Multiply(A=OneMinus(None=Saturate(None=@n169)), B=OneMinus(None=@n168))
  n171 = Constant() {r=1}
  NamedRerouteDeclaration(None=QualitySwitch(
    Default=ShadingPathSwitch(Default=@n170, Deferred=@n170, Mobile=@n171) {desc=Migrated from FeatureLevelSwitch}
    Low=@n171
    High=@n170
    Medium=@n171
   )) {name=DissolveMask, nodeColor=<rgba 0.6875 0 1 1>}
  n172 = Constant()
  n173 = Saturate(None=Subtract(A=@n134, B=LinearInterpolate(Alpha=@n137) {constA=-0.2}))
  n174 = VectorParameter() {group=IntersectingEdge, parameterName=IntersectingEdgeColor, defaultValue=<rgba 1 1 1 1>}
  n175 = NamedRerouteUsage()
  n176 = Add(A=Multiply(A=@n173) {constB=30}, B=LinearInterpolate(Alpha=Multiply(A=@n175, B=@n175)) {constA=-1, constB=3}) {constB=0}
  n177 = Multiply(A=Multiply(A=@n173, B=Multiply(A=@n174.RGB, B=@n176)), B=Power(Base=NamedRerouteUsage()) {constExponent=4})
  NamedRerouteDeclaration(None=QualitySwitch(
    Default=ShadingPathSwitch(Default=@n177, Deferred=@n177, Mobile=@n172) {desc=Migrated from FeatureLevelSwitch}
    Low=@n172
    High=@n177
    Medium=@n172
   )) {name=DeplyingWaveColor, nodeColor=<rgba 0.2891 1 0 1>}
  n178 = Constant3Vector() {desc=Random, constant=<rgba -0.3864 0.6 0.4464 0>}
  n179 = Multiply(A=Time(), B=ScalarParameter() {defaultValue=0.1, group=Hummoon, parameterName=HunmoonNoise_Frequence, sortPriority=1})
  n180 = WorldPosition() {worldPositionShaderOffset=WPT_ExcludeAllShaderOffsets}
  n181 = LocalPosition()
  n182 = MaterialFunctionCall(Float3=Multiply(
    A=StaticSwitchParameter(True=@n180.XYZ, False=@n181.XYZ) {defaultValue=true, group=Hummoon, parameterName=Enable_Randomize}
    B=VectorParameter() {group=Hummoon, parameterName=Noise_Size, defaultValue=<rgba 0.01 0.04 0.1 1>}.RGB
   )) {materialFunction=&BreakOutFloat3Components}
  n183 = Sine(None=Multiply(A=@n178.R, B=Add(A=@n179, B=@n182.R))) {period=6.283}
  n184 = Sine(None=Multiply(A=@n178.G, B=Add(A=@n179, B=@n182.G))) {period=6.283}
  n185 = Sine(None=Multiply(A=@n178.B, B=Add(A=@n179, B=@n182.B))) {period=6.283}
  n186 = MaterialFunctionCall(Float3=Multiply(
    A=MaterialFunctionCall(X=@n183, Y=@n184, Z=@n185) {materialFunction=&MakeFloat3}.Result
    B=ScalarParameter() {defaultValue=0.1, group=Hummoon, parameterName=HunmoonNoise_Amplitude, sortPriority=4}
   ) {desc=wind factor}) {materialFunction=&BreakOutFloat3Components}
  NamedRerouteDeclaration(None=MaterialFunctionCall(X=Add(A=@n186.G, B=@n186.G), Y=Add(A=@n186.R, B=@n186.R), Z=Add(A=@n186.R, B=@n186.R)) {materialFunction=&MakeFloat3}.Result) {name=WaveNoise, nodeColor=<rgba 0 1 0.3203 1>}
  n187 = NamedRerouteUsage()
  n188 = PixelDepth()
  n189 = Constant() {r=-100}
  n190 = Saturate(None=Divide(A=Add(A=@n188.R, B=@n189), B=ScalarParameter() {defaultValue=300, parameterName=NearFade}))
  n191 = Multiply(A=Add(A=@n187), B=@n190)
  n192 = LinearInterpolate(
   A=ScalarParameter() {defaultValue=0.5, group=Opacity, parameterName=OpacityInside}
   B=ScalarParameter() {defaultValue=1, group=Opacity, parameterName=OpacityOutside}
   Alpha=Saturate(None=TwoSidedSign())
  )
  n193 = Multiply(A=NamedRerouteUsage(), B=@n192)
  n194 = NamedRerouteUsage()
  n195 = Multiply(A=Add(A=@n187, B=LinearInterpolate(A=@n193, B=@n192, Alpha=@n194)), B=@n190)
  n196 = Multiply(
   A=Multiply(A=ShadingPathSwitch(Default=@n191, Deferred=@n195, Mobile=@n191) {desc=Migrated from FeatureLevelSwitch}, B=NamedRerouteUsage())
   B=NamedRerouteUsage()
  )
  n197 = DepthFade(Opacity=@n196, FadeDistance=ScalarParameter() {defaultValue=20, parameterName=Depthfade})
  NamedRerouteDeclaration(None=Multiply(
    A=ShadingPathSwitch(Default=@n197, Mobile=@n196) {desc=Migrated from FeatureLevelSwitch}
    B=ScalarParameter() {defaultValue=1, parameterName=OverallOpacity}
   ) {desc=OP}) {name=OutOpacity, nodeColor=<rgba 0 1 0.3672 1>}
  n198 = NamedRerouteUsage()
  n199 = Constant()
  n200 = LinearInterpolate(
   B=ScalarParameter() {defaultValue=0.02, group=HitWave, parameterName=Hitwave-DistortAmt3}
   Alpha=QualitySwitch(
    Default=ShadingPathSwitch(Default=@n82, Deferred=@n82, Mobile=@n199) {desc=Migrated from FeatureLevelSwitch}
    Low=@n199
    High=@n82
    Medium=@n199
   )
  )
  NamedRerouteDeclaration(None=Multiply(A=Multiply(A=@n198, B=@n200), B=ScalarParameter() {defaultValue=1, parameterName=HitEffectLerp})) {name=Hitwave_Distort, nodeColor=<rgba 0 1 0.1094 1>}
  n201 = TextureSample!(UVs=Panner(Coordinate=Multiply(A=@n106) {constB=2}) {speedX=-0.2, speedY=0.1}, Tex, "DDX(UVs)"=@n107, "DDY(UVs)"=@n107, Apply View MipBias) {mipValueMode=TMVM_Derivative, samplerSource=SSM_Wrap_WorldGroupSettings, texture=&T_godray_Noise}
  n202 = NamedRerouteUsage()
  n203 = Multiply(A=@n201.R, B=OneMinus(None=Power(Base=@n202)))
  n204 = LinearInterpolate(Alpha=NamedRerouteUsage()) {constA=2, constB=0.5}
  NamedRerouteDeclaration(None=Multiply(A=Power(Base=@n203, Exp=@n204), B=NamedRerouteUsage())) {name=HM_FineLine, nodeColor=<rgba 1 0.8203 0 1>}
  n205 = MaterialFunctionCall() {materialFunction=/Game/Packages/Fortress_SharedMaterials/Base_Material_Functions/BaseCharacterShaderFunctions/Audio/MF_AudioReactivity}
  n206 = NamedRerouteUsage()
  n207 = Multiply(A=@n205.Amplitude, B=@n206)
  n208 = Power(Base=@n207, Exp=ScalarParameter() {defaultValue=3, group=Audio Reactivity, parameterName=Audio_Amplitude_Exp})
  n209 = Power(Base=Multiply(A=@n205.Spectrum, B=@n206), Exp=ScalarParameter() {defaultValue=5, group=Audio Reactivity, parameterName=Audio_Spectrum_Exp}) {constExponent=4}
  n210 = MaterialFunctionCall(Float4=Add(A=Multiply(A=@n208, B=@n209) {constB=0.05}, B=@n209)) {desc=Bass/Low/Mid/High, materialFunction=&BreakOutFloat4Components}
  n211 = VectorParameter() {group=Audio Reactivity, parameterName=Spectrum_Intensity, defaultValue=<rgba 1 1 1 1>}
  NamedRerouteDeclaration(None=Multiply(A=@n210.R, B=@n211.R)) {name=BassFreq, desc=WPO, nodeColor=<rgba 1 0 1 1>}
  n212 = Multiply(
   A=VectorParameter() {group=IntersectingEdge, parameterName=IntersectingEdgeColor, defaultValue=<rgba 1 1 1 1>}.RGB
   B=Add(A=NamedRerouteUsage(), B=ScalarParameter() {defaultValue=1.5, group=IntersectingEdge, parameterName=IntBrightness})
  )
  n213 = Saturate(None=TwoSidedSign())
  NamedRerouteDeclaration(None=Multiply(A=Multiply(A=@n212, B=@n95), B=LinearInterpolate(Alpha=@n213) {constA=0.25, constB=0.5})) {name=IntersectingEdge_Colored, nodeColor=<rgba 0 0.2031 1 1>}
  NamedRerouteDeclaration(None=Multiply(A=@n210.G, B=@n211.G)) {name=LowFreq, desc=IntersectingEdge, nodeColor=<rgba 1 0 1 1>}
  NamedRerouteDeclaration(None=Multiply(A=@n210.B, B=@n211.B)) {name=MidFreq, nodeColor=<rgba 1 0 1 1>}
  NamedRerouteDeclaration(None=Multiply(A=@n210.A, B=@n211.A)) {name=HighFreq, nodeColor=<rgba 1 0 0.9913 1>}
  NamedRerouteDeclaration(None=@n208) {name=OverallAmp, nodeColor=<rgba 1 0 0.9609 1>}
  NamedRerouteDeclaration(None=TextureObject() {texture=&T_godray_Noise}) {name=NoiseTex, nodeColor=<rgba 0 0.4475 1 1>}
  NamedRerouteDeclaration(None=Step(X=@n207) {constX=0, constY=0.1}) {name=AudioTriggered, nodeColor=<rgba 1 0 0.1641 1>}
  n214 = NamedRerouteUsage()
  n215 = Constant()
  n216 = TextureSample(UVs=Panner(Coordinate=@n214) {speedY=-0.0125}, "DDX(UVs)"=@n215, "DDY(UVs)"=@n215) {mipValueMode=TMVM_Derivative, samplerSource=SSM_Wrap_WorldGroupSettings, texture=&T_godray_Noise}
  n217 = Constant()
  n218 = TextureSample(UVs=Add(A=Panner(Coordinate=@n214) {speedY=0.1}, B=Multiply(A=@n216.R) {constB=0.1}), "DDX(UVs)"=@n217, "DDY(UVs)"=@n217) {mipValueMode=TMVM_Derivative, samplerSource=SSM_Wrap_WorldGroupSettings, texture=&T_Tile_RainbowColor}
  n219 = Desaturation(None=Normalize(VectorInput=@n218.RGB), Fraction=ScalarParameter() {defaultValue=-0.9536, parameterName=DES})
  n220 = LinearInterpolate(
   A=VectorParameter() {group=Color, parameterName=BubbleThinFilmColor_Near, defaultValue=<rgba 0.4696 0.4696 2 1>}.RGB
   B=VectorParameter() {group=Color, parameterName=BubbleThinFilmColor_Far, defaultValue=<rgba 2 0.5 0.1 1>}.RGB
   Alpha=NamedRerouteUsage()
  )
  n221 = ScalarParameter() {defaultValue=1, group=Color, parameterName=BubbleThinFilmIntensity}
  n222 = Multiply(A=@n221) {constB=2}
  n223 = NamedRerouteUsage()
  NamedRerouteDeclaration(None=Multiply(A=Multiply(A=@n219, B=@n220), B=LinearInterpolate(A=@n221, B=@n222, Alpha=@n223) {constB=2})) {name=BubbleThinFilmColor, nodeColor=<rgba 1 0.8906 0 1>}
  NamedRerouteDeclaration(None=SmoothStep(Value=@n207) {constMax=0.9, constMin=0.1}) {name=AudioOnTransition, nodeColor=<rgba 1 0 0.1641 1>}
  n224 = Length(None=Subtract(A=CameraPositionWS(), B=WorldPosition().XYZ)) {is=class MaterialExpressionLength}
  n225 = Divide(A=@n224, B=Multiply(A=ScalarParameter() {defaultValue=24, group=Toyfox, parameterName=CenterOpacityDistStartFade}) {constB=100}) {constB=1000}
  NamedRerouteDeclaration(None=Saturate(None=Power(Base=@n225) {constExponent=4})) {name=DistToFade_Opacity, nodeColor=<rgba 0.4531 1 0 1>}
  n226 = Saturate(None=TwoSidedSign())
  n227 = VertexNormalWS()
  n228 = Multiply(A=@n227) {constB=-1}
  n229 = Divide(
   A=LinearInterpolate(Alpha=@n226) {constA=0.75, constB=0.5}
   B=DotProduct(A=CameraVectorWS(), B=LinearInterpolate(A=@n228, B=@n227, Alpha=@n226))
  ) {constA=0.75}
  n230 = LinearInterpolate(Alpha=@n226) {constA=10, constB=4}
  NamedRerouteDeclaration(None=Saturate(None=Power(Base=@n229, Exp=@n230) {constExponent=16})) {name=CenterMasking, nodeColor=<rgba 0 0.2734 1 1>}
  n231 = MaterialFunctionCall(Float3=@n43) {materialFunction=&BreakOutFloat3Components}
  n232 = MaterialFunctionCall(Float2=Normalize(VectorInput=MaterialFunctionCall(X=@n231.R, Y=@n231.R) {materialFunction=&MakeFloat2}.Result)) {materialFunction=&BreakOutFloat2Components}
  n233 = Divide(
   A=Arctangent2!(Y=@n232.R, X=@n232.R)
   B=MaterialFunctionCall(Multiplier=Constant() {r=2}) {materialFunction=&Pi}.Result
  )
  n234 = Constant() {r=1}
  n235 = Multiply(A=Divide(
    A=Arcsine!(None=@n231.B)
    B=MaterialFunctionCall(Multiplier=@n234) {materialFunction=&Pi}.Result
   )) {constB=-1}
  NamedRerouteDeclaration(None=Add(A=MaterialFunctionCall(X=@n233, Y=@n235) {materialFunction=&MakeFloat2}.Result) {constB=0.5}) {name=SphericalProjectionUV, nodeColor=<rgba 0.2187 1 0 1>}
  NamedRerouteDeclaration(None=Saturate(None=@n42)) {name=FlareLines, nodeColor=<rgba 1 0.6094 0 1>}
  n236 = NamedRerouteUsage()
  n237 = Panner(Coordinate=@n236) {speedY=-0.002}
  n238 = NamedRerouteUsage()
  n239 = ScalarParameter() {defaultValue=0.15, parameterName=HM_Distort_Intensity}
  n240 = Add(A=Add(A=Multiply(A=@n238, B=@n239) {constB=0.15}, B=NamedRerouteUsage()), B=NamedRerouteUsage())
  n241 = Add(A=Subtract(A=ComponentMask(None=@n237) {g=true}, B=ComponentMask(None=@n237) {r=true}), B=@n240)
  n242 = Frac(None=Divide(A=Multiply(A=@n241) {constB=4}, B=ScalarParameter() {defaultValue=0.1, group=Hummoon, parameterName=TopographyLineScale}))
  n243 = Abs(None=Subtract(A=Multiply(A=@n242) {constB=2}))
  NamedRerouteDeclaration(None=@n243) {name=HM_LinesMask_OLD, nodeColor=<rgba 1 0 0.375 1>}
  NamedRerouteDeclaration(None=Saturate(None=Multiply(A=@n108.R, B=@n201.R))) {name=RawCloudNoise, nodeColor=<rgba 0.7813 0 1 1>}
  n244 = Panner(Coordinate=@n236) {speedX=-0.05}
  n245 = Add(A=Add(A=ComponentMask(None=@n244) {g=true}, B=ComponentMask(None=@n244) {r=true}), B=@n240)
  n246 = Multiply(A=Frac(None=Multiply(A=@n245) {constB=3})) {constB=2}
  n247 = Power(Base=Abs(None=Subtract(A=@n246))) {constExponent=3}
  NamedRerouteDeclaration(None=Saturate(None=Multiply(A=@n243, B=@n247))) {name=HM_LineThicknessFactor_Scan_OLD, nodeColor=<rgba 0.0547 1 0 1>}
  n248 = Divide(A=@n224, B=Multiply(A=ScalarParameter() {defaultValue=3, group=Toyfox, parameterName=LinesDistStartFade}) {constB=100})
  NamedRerouteDeclaration(None=Saturate(None=Power(Base=@n248) {constExponent=4})) {name=DistToFade_Line, nodeColor=<rgba 0.4531 1 0 1>}
  n249 = Add(A=@n186.R, B=@n186.R)
  NamedRerouteDeclaration(None=Multiply(A=Add(A=@n249, B=@n186.B), B=ScalarParameter() {defaultValue=0.2, group=Hummoon, parameterName=HunmoonNoise_Strength, sortPriority=3})) {name=HM_SinwaveNoise, nodeColor=<rgba 0 1 0.5547 1>}
  n250 = NamedRerouteUsage()
  n251 = Constant2Vector() {g=2, r=2}
  n252 = Constant()
  n253 = TextureSample(UVs=Panner(Coordinate=Multiply(A=@n250, B=@n251)) {speedY=-0.01}, "DDX(UVs)"=@n252, "DDY(UVs)"=@n252) {
   mipValueMode=TMVM_Derivative
   samplerSource=SSM_Wrap_WorldGroupSettings
   samplerType=SAMPLERTYPE_LinearGrayscale
   texture=&T_SwirlNoise_N_height
  }
  NamedRerouteDeclaration(None=@n253.R) {name=HMDistortion, nodeColor=<rgba 0.3594 1 0 1>}
  n254 = ComponentMask(None=NamedRerouteUsage()) {g=true}
  NamedRerouteDeclaration(None=Sine(None=Multiply(A=@n254) {constB=0.5})) {name=PolarMask, nodeColor=<rgba 0 1 0.4375 1>}
  n255 = VertexNormalWS()
  n256 = TwoSidedSign()
  n257 = DotProduct(A=CameraVectorWS(), B=LinearInterpolate(A=Multiply(A=@n255) {constB=-1}, B=@n255, Alpha=Saturate(None=@n256)))
  NamedRerouteDeclaration(None=Power(Base=OneMinus(None=@n257)) {constExponent=8}) {name=TwoSidedFresnel, nodeColor=<rgba 0.1016 0 1 1>}
  NamedRerouteUsage()
  NamedRerouteUsage()
  NamedRerouteUsage()
  NamedRerouteUsage()
  n258 = Multiply(A=@n243)
  n259 = ScalarParameter() {defaultValue=0.1, parameterName=HMLineWidth}
  n260 = LinearInterpolate(Alpha=NamedRerouteUsage()) {constB=0.3}
  Saturate(None=SmoothStep(Min=Subtract(A=@n243, B=@n258), Max=Add(A=@n243, B=@n258), Value=Add(A=@n259, B=@n260) {constB=0.1})) {desc=Fine Line}
  Subtract(A=@n7)
 ]
}
